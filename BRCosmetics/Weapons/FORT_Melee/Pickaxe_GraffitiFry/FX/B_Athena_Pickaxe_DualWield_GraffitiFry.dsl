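; Dual-wield GraffitiFry pickaxe Blueprint that tells its swing effects which swing animation section is playing.
; On each right or left swing it runs the parent logic, then derives a section index from the instigator's swing montage.
; It writes that index to the SwingSectionIndex int parameter on the main or offhand swing VFX component respectively.
; BeginPlay and Tick are empty, and the construction script only calls its parent.
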(event EventBeginPlay)

(event EventSwingRight
  (|CallParentFunction)
  (bind _section_index (Utilities|SetupSwingMontageSectionIndex (Variables|Actor|GetInstigator))
    (:Out
      (Effects|Components|ParticleSystem|SetIntParameter (CallFunction|GetSwingVFXComponent) "SwingSectionIndex" _section_index))))

(event EventTick
  (params (DeltaSeconds :type Float%20%28single-precision%29)))

(event EventSwingLeft
  (|CallParentFunction)
  (bind _section_index (Utilities|SetupSwingMontageSectionIndex (Variables|Actor|GetInstigator))
    (:Out
      (Effects|Components|ParticleSystem|SetIntParameter (CallFunction|GetSwingOffhandVFXComponent) "SwingSectionIndex" _section_index))))

(fn ConstructionScript
  (|CallParentFunction))
