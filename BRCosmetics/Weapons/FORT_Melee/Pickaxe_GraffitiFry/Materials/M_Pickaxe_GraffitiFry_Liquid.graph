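# Liquid material for the GraffitiFry pickaxe cosmetic, a Default Lit surface usable on skeletal meshes and mesh deformers.
# It uses material attributes, with a SetMaterialAttributes node feeding the MF_Potion material function, which holds the actual liquid shading logic.
# The graph itself adds no other nodes or parameters of its own.
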Material /BRCosmetics/Weapons/FORT_Melee/Pickaxe_GraffitiFry/Materials/M_Pickaxe_GraffitiFry_Liquid
graph {
 properties {
  bUseMaterialAttributes=true
  bUsedWithMeshDeformer=true
  bUsedWithSkeletalMesh=true
  pixelDepthOffsetMode=PDOM_Legacy
  usedShadingModels=Default Lit
 }
 nodes [MaterialFunctionCall!(In=SetMaterialAttributes!(MaterialAttributes)) {materialFunction=&/MF_Potion}]
}
