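# Physics asset for the GraffitiFry two-handed pickaxe cosmetic mesh, providing its simplified collision shape.
# It holds a single capsule body on the Melee_TwoHanded_Handle bone, offset upward and rotated along the weapon's length.
# It has no constraints, so the weapon is represented as one rigid piece.
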PhysicsAsset /BRCosmetics/Weapons/FORT_Melee/Pickaxe_GraffitiFry/Meshes/GraffitiFry_Axe_PhysicsAsset
bodies {Melee_TwoHanded_Handle [{shapeType=Capsule, length=35.5424, radius=6.9655, center=<0.0336 0 23.9646>, rotation=<rot 0 -90 0.1331>}]}
constraints []
